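# Post-process AnimBlueprint for the Stamina character head that shrinks head and neck while the local player aims, apparently to keep the crosshair clear.
# Each update derives scaleDownAlpha from local pawn, targeting and visible-weapon checks.
# Weapon visibility excludes holstered weapons and TrapPlacement, RocketLauncher and GoingCommando core animations.
# The AnimGraph uses that bool to briefly blend non-uniform bone-space scales onto head and neck_02 over the input pose.
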AnimBlueprint /BRCosmetics/Characters/Player/Male/Medium/Heads/M_MED_Stamina_Head/Meshes/Stamina_PostProcess
target_skeleton=/Game/Characters/Player/Common/Fortnite_Base_Head/Fortnite_Base_Head_Export_Skeleton
graphs [
 AnimGraph {nodes [
   n1 = Get scaleDownAlpha() -> (scaleDownAlpha:bool)
   n2 = "Transform (Modify) Bone - Bone: head"!(
    ComponentPose=Local To Component(LocalPose=Input Pose())
    bAlphaBoolEnabled:bool=@n1
    Translation="0,0,0"
    Rotation="0,0,0"
    Scale="0.85,1,1"
   ) -> (Pose) {
    alphaInputType=Bool
    scaleMode=BMM_Replace
    scaleSpace=BCS_BoneSpace
    alphaBoolBlend {blendInTime=0.1, blendOption=Linear, blendOutTime=0.12}
    boneToModify {boneName=head}
    scale {x=0.85}
   }
   n3 = "Transform (Modify) Bone - Bone: neck_02"!(ComponentPose=@n2, bAlphaBoolEnabled:bool=@n1, Translation="0,0,0", Rotation="0,0,0", Scale="1,0.85,0.85") -> (Pose) {
    alphaInputType=Bool
    scaleMode=BMM_Replace
    scaleSpace=BCS_BoneSpace
    alphaBoolBlend {blendInTime=0.1, blendOption=Linear, blendOutTime=0.12}
    boneToModify {boneName=neck_02}
    scale {y=0.85, z=0.85}
   }
   Output Pose(Result=Component To Local(ComponentPose=@n3)) {layerGroup=None}
  ]}
 EventGraph {nodes [
   Event BlueprintUpdateAnimation()
   > n1 = Sequence()
   > Set FortPawn(FortPawn:Fort Pawn Object Reference=TryGetFortPawn()) -> (Output_Get:Fort Pawn Object Reference)
   n2 = Get FortPawn() -> (FortPawn:Fort Pawn Object Reference)
   > n3 = Is Valid(InputObject=@n2)
   Set weapon_visible(execute=@n3.Is Not Valid, weapon_visible:bool=false) -> (Output_Get:bool)
   n4 = Get CurrentWeapon!(self:Fort Pawn Object Reference=@n2) -> (CurrentWeapon:Fort Weapon Object Reference)
   n5 = GetCoreAnimation!(self:Fort Weapon Object Reference=@n4) -> (ReturnValue:EFortWeaponCoreAnimation Enum)
   n6 = Set notCoverCrosshair(
    execute=@n3.Is Valid
    notCoverCrosshair:bool=OR Boolean!(
     A="Equal (Enum)"(A:EFortWeaponCoreAnimation Enum=@n5, B:EFortWeaponCoreAnimation Enum=TrapPlacement)
     B="Equal (Enum)"(A=@n5, B=RocketLauncher)
     C:bool="Equal (Enum)"(A=@n5, B=GoingCommando)
    ) -> (ReturnValue)
   ) -> (Output_Get:bool)
   > n7 = Branch(Condition=@n6)
   > Set weapon_visible(weapon_visible=false)
   n8 = IsWeaponHolstered!(self:Fort Pawn Object Reference=@n2) -> (ReturnValue:bool)
   Set weapon_visible(execute=@n7.else, weapon_visible=AND Boolean(A=IsValid(Object=@n4), B=NOT Boolean(A=@n8)))
   n9 = Set isLocalPlayer(execute=@n1.then_1, isLocalPlayer:bool=IsLocalPawn!(Actor:Actor Object Reference=GetOwningActor()) -> (IsLocalPawn:bool)) -> (Output_Get:bool)
   > n10 = Branch(Condition=AND Boolean!(A=@n9, B=Get bIsTargeting(), C:bool=Get weapon_visible() -> (weapon_visible:bool)) -> (ReturnValue))
   > Set scaleDownAlpha(scaleDownAlpha:bool=true) -> (Output_Get:bool)
   Set scaleDownAlpha(execute=@n10.else, scaleDownAlpha=false)
  ]}
]
variables [
 FortPawn {type=FortPawn, default=null}
 isLocalPlayer {type=boolean}
 notCoverCrosshair {type=boolean}
 scaleDownAlpha {type=boolean}
 weapon_visible {type=boolean}
]
